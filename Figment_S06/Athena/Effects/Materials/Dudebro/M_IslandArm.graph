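# Translucent, two-sided unlit effect material for island "arm" spline meshes, drawing a glowing, flowing smoky tendril.
# SplineThicken sets tapered width, while panning curl noise, masked along the length, adds world-position wobble.
# A noise-distorted smoke trail, brightened by a looping sphere-mask sweep, drives a two-color gradient.
# That trail plus a second panning inky-smoke layer forms opacity, shaped by edge and length masks and distance fade.
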Material /Figment_S06/Athena/Effects/Materials/Dudebro/M_IslandArm
graph {
 properties {blendMode=BLEND_Translucent, pixelDepthOffsetMode=PDOM_Legacy, shadingModel=MSM_Unlit, twoSided=true, usedShadingModels=Unlit}
 nodes [
  n1 = Constant() {r=0.5}
  n2 = SphereMask(
   A=MaterialFunctionCall(Float2=TextureCoordinate()) {materialFunction=&BreakOutFloat2Components}.G
   B=@n1
   Radius=@n1
   Hardness=ScalarParameter() {defaultValue=0.55, parameterName=UMaskHardness}
  )
  n3 = Add(A=WorldPosition().XYZ, B=Multiply(A=Time(), B=VectorParameter() {parameterName=CurlPan, defaultValue=<rgba 10 0 0 0>}.RGB))
  n4 = VectorNoise(World Position=Divide(
    A=Transform(None=@n3) {transformSourceType=TRANSFORMSOURCE_World, transformType=TRANSFORM_Local}
    B=VectorParameter() {parameterName=CurlTileScale, defaultValue=<rgba 100 100 100 0>}.RGB
   )) {noiseFunction=VNF_CurlALU}
  n5 = VectorParameter() {parameterName=CurlIntensity, defaultValue=<rgba 90 90 90 0>}
  Add(
   A=MaterialFunctionCall(
    WidthBase=ScalarParameter() {defaultValue=100, parameterName=SplineWidthBase}
    WidthTip=ScalarParameter() {defaultValue=800, parameterName=SplineWidthTip}
   ) {materialFunction=&SplineThicken}.WorldPositionOffset
   B=Multiply(A=@n2, B=Multiply(A=@n4, B=@n5.RGB))
  )
  n6 = AppendVector(A=ScalarParameter() {defaultValue=1.5, parameterName=TileU_B}, B=Constant() {r=1})
  n7 = TextureCoordinate()
  n8 = MaterialFunctionCall(Float2=@n7) {materialFunction=&BreakOutFloat2Components}
  n9 = ScalarParameter() {defaultValue=0.5, parameterName=V Exponent}
  n10 = MaterialFunctionCall(XY=MaterialFunctionCall(X=@n7, Y=Power(Base=@n8.G, Exp=@n9)) {materialFunction=&MakeFloat2}.Result) {materialFunction=&Swizzle}
  n11 = ScalarParameter() {defaultValue=10, parameterName=Random}
  n12 = Constant()
  n13 = TextureSample(UVs=Panner() {bFractionalPart=true, speedY=0.1}) {samplerType=SAMPLERTYPE_LinearColor, texture=/Game/Effects/Fort_Effects/Gameplay/SmokeDustExplosions/Textures/T_NoisePack_LinearColor}
  n14 = ComponentMask(None=@n13.RGB) {g=true, r=true}
  n15 = ScalarParameter() {defaultValue=0.05, parameterName=UVDistortIntensity}
  n16 = Constant()
  n17 = Add(
   A=Add(A=@n10.YX, B=AppendVector(A=@n11, B=@n12))
   B=Multiply(A=ConstantBiasScale(None=@n14) {bias=-0.5, scale=2}, B=AppendVector(A=@n15, B=@n16))
  )
  n18 = ScalarParameter() {defaultValue=0.24, parameterName=PanSpeedU_B}
  n19 = Time()
  n20 = TextureSample(UVs=Panner(Coordinate=Multiply(A=@n6, B=@n17), Time=Multiply(A=@n18, B=@n19)) {bFractionalPart=true, speedX=1}) {samplerType=SAMPLERTYPE_Grayscale, texture=&T_Smoke_Trail_Soft}
  n21 = TextureCoordinate()
  n22 = Time() {period=1}
  n23 = ScalarParameter() {defaultValue=0.5, parameterName=SweepSpeed}
  n24 = Divide(B=@n23)
  n25 = Divide(A=Fmod(A=Multiply(A=@n22, B=@n23), B=@n24), B=@n24)
  n26 = SphereMask(
   A=Add(A=ConstantBiasScale(None=@n20.RGB) {bias=-0.5, scale=0.2}, B=ComponentMask(None=@n21) {g=true})
   B=ConstantBiasScale(None=OneMinus(None=@n25)) {bias=-0.2, scale=1.3}
   Radius=ScalarParameter() {defaultValue=0.15, parameterName=SweepRadius}
   Hardness=ScalarParameter() {defaultValue=0.05, parameterName=SweepHardness}
  )
  n27 = Add(A=@n20.RGB, B=Multiply(A=Multiply(A=@n20.RGB, B=@n26), B=ScalarParameter() {defaultValue=2, parameterName=SweepScale}))
  n28 = ScalarParameter() {defaultValue=1.4, parameterName=Color Gain}
  LinearInterpolate(
   A=VectorParameter() {parameterName=ColorLow, defaultValue=<rgba 0.1117 0.0868 0.5208 0>}.RGB
   B=VectorParameter() {parameterName=ColorHigh, defaultValue=<rgba 1.9533 0.1354 2 0>}.RGB
   Alpha=Power(Base=Multiply(A=@n27, B=@n28), Exp=ScalarParameter() {defaultValue=3.1, parameterName=Color Exponent})
  )
  n29 = AppendVector(A=ScalarParameter() {defaultValue=0.9, parameterName=TileU_A}, B=Constant() {r=1})
  n30 = ScalarParameter() {defaultValue=0.72, parameterName=PanSpeedU_A}
  n31 = Time()
  n32 = TextureSample(UVs=Panner(Coordinate=Multiply(A=@n29, B=@n10.YX), Time=Multiply(A=@n30, B=@n31)) {bFractionalPart=true, speedX=1}) {texture=/Game/Effects/Fort_Effects/Textures/Tile/T_Inky_Smoke_Tile}
  n33 = ScalarParameter() {defaultValue=0.6, parameterName=BGScale}
  n34 = SphereMask(
   A=MaterialFunctionCall(Float2=TextureCoordinate()) {materialFunction=&BreakOutFloat2Components}.R
   B=Constant() {r=0.5}
   Radius=ScalarParameter() {defaultValue=0.5, parameterName=VMaskRadius}
   Hardness=ScalarParameter() {defaultValue=0.6, parameterName=VMaskHardness}
  )
  n35 = Multiply(A=Add(A=Multiply(A=@n32.RGB, B=@n33), B=@n20.RGB), B=@n34)
  n36 = Multiply(A=ScalarParameter() {defaultValue=0.2, parameterName=BackOpacity}, B=@n34)
  MaterialFunctionCall(
   Opacity In=Multiply(A=Add(A=@n35, B=@n36), B=@n2)
   Position=WorldPosition().XYZ
   Fade_Start_Distance=ScalarParameter() {defaultValue=500, parameterName=Fade_Start_Distance}
   Fade_End_Distance=ScalarParameter() {defaultValue=150, parameterName=Fade_End_Distance}
  ) {materialFunction=&MF_Distance_SyncedFade_FX}
 ]
}
