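# Translucent unlit material for an island base effect, producing a two-color animated electric glow.
# Two panning electric-noise layers, one mirrored in U, use V-exponent-warped UVs with a Random U offset, each distorted by panning smoke noise.
# Their sum plus a V-gradient glow drives the color blend.
# The same terms, boosted, are multiplied by gradient-varied Fresnel and a V-gradient fade to shape opacity.
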Material /Figment_S06/Athena/Effects/Materials/Dudebro/M_IslandBase
graph {
 properties {blendMode=BLEND_Translucent, pixelDepthOffsetMode=PDOM_Legacy, shadingModel=MSM_Unlit, usedShadingModels=Unlit}
 nodes [
  n1 = TextureSample(UVs=Panner(Coordinate=TextureCoordinate() {vTiling=0.125}) {bFractionalPart=true, speedY=0.02}) {samplerType=SAMPLERTYPE_LinearColor, texture=/Game/Effects/Fort_Effects/Gameplay/SmokeDustExplosions/Textures/T_NoisePack_LinearColor}
  n2 = ConstantBiasScale(None=ComponentMask(None=@n1.RGB) {g=true, r=true}) {bias=-0.5, scale=2}
  n3 = ScalarParameter() {defaultValue=0.06, parameterName=UVDistortIntensity_2}
  n4 = Constant2Vector() {g=1, r=-1}
  n5 = TextureCoordinate()
  n6 = MaterialFunctionCall(Float2=@n5) {materialFunction=&BreakOutFloat2Components}
  n7 = ScalarParameter() {defaultValue=0.5, parameterName=V Exponent}
  n8 = Add(
   A=MaterialFunctionCall(X=@n5, Y=Power(Base=@n6.G, Exp=@n7)) {materialFunction=&MakeFloat2}.Result
   B=AppendVector(A=ScalarParameter() {parameterName=Random}, B=Constant())
  )
  n9 = Multiply(
   A=AppendVector(A=ScalarParameter() {defaultValue=3, parameterName=TileU_2}, B=ScalarParameter() {defaultValue=0.4665, parameterName=TileV_2})
   B=Add(A=Multiply(A=@n2, B=@n3), B=Multiply(A=@n4, B=@n8))
  )
  n10 = Multiply(A=ScalarParameter() {defaultValue=-0.0357, parameterName=PanSpeed_2}, B=Time())
  n11 = TextureSample(UVs=Panner(Coordinate=TextureCoordinate() {vTiling=0.3}) {bFractionalPart=true, speedY=0.02}) {samplerType=SAMPLERTYPE_LinearColor, texture=&T_NoisePack_LinearColor}
  n12 = ConstantBiasScale(None=ComponentMask(None=@n11.RGB) {g=true, r=true}) {bias=-0.5, scale=2}
  n13 = ScalarParameter() {defaultValue=0.05, parameterName=UVDistortIntensity_1}
  n14 = Multiply(
   A=AppendVector(A=ScalarParameter() {defaultValue=2, parameterName=TileU_1}, B=ScalarParameter() {defaultValue=0.35, parameterName=TileV_1})
   B=Add(A=@n8, B=Multiply(A=@n12, B=@n13))
  )
  n15 = Multiply(A=ScalarParameter() {defaultValue=-0.06, parameterName=PanSpeed_1}, B=Time())
  n16 = Add(
   A=TextureSample(UVs=Panner(Coordinate=@n9, Time=@n10) {bFractionalPart=true, speedY=1}) {samplerType=SAMPLERTYPE_LinearColor, texture=/Game/Characters/Player/Male/Medium/Bodies/M_MED_SRDrift/Textures/T_ElectricNoise}.RGB
   B=TextureSample(UVs=Panner(Coordinate=@n14, Time=@n15) {bFractionalPart=true, speedY=1}) {samplerType=SAMPLERTYPE_LinearColor, texture=&T_ElectricNoise}.RGB
  )
  n17 = ComponentMask(None=TextureCoordinate()) {g=true}
  n18 = ScalarParameter() {defaultValue=6, parameterName=GlowGradientExponent}
  n19 = Saturate(None=Multiply(A=Power(Base=@n17, Exp=@n18), B=ScalarParameter() {defaultValue=4, parameterName=GlowGradientScale}))
  LinearInterpolate(
   A=VectorParameter() {parameterName=ColorLow, defaultValue=<rgba 0.1117 0.0868 0.5208 0>}.RGB
   B=VectorParameter() {parameterName=ColorHigh, defaultValue=<rgba 1.9533 0.1354 2 0>}.RGB
   Alpha=Saturate(None=Add(A=@n16, B=@n19))
  )
  n20 = Add(A=@n16, B=ScalarParameter() {defaultValue=0.5, parameterName=BGBoost})
  n21 = LinearInterpolate(
   A=ScalarParameter() {defaultValue=1.5, parameterName=FresnelBrightness}
   B=ScalarParameter() {defaultValue=2, parameterName=FresnelBrightnessB}
   Alpha=@n19
  )
  n22 = ScalarParameter() {defaultValue=3, parameterName=FresnelExponent}
  n23 = Power(Base=ComponentMask(None=TextureCoordinate()) {g=true}, Exp=ScalarParameter() {defaultValue=2.5, parameterName=OpacityGradientExponent})
  n24 = ScalarParameter() {defaultValue=2.5, parameterName=OpacityGradientScale}
  Multiply(
   A=Multiply(
    A=Add(A=@n20, B=@n19)
    B=MaterialFunctionCall(Brightness=@n21, Exponent=@n22) {materialFunction=&MF_Fresnel_FX}.Fresnel
   )
   B=Saturate(None=Multiply(A=@n23, B=@n24))
  )
 ]
}
